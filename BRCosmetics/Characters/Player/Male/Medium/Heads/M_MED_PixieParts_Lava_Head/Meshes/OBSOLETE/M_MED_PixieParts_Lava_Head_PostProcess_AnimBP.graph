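# Post-process AnimBP for the PixieParts Lava male head, in an OBSOLETE folder, that adds movement-driven Control Rig animation and dampens head and neck motion.
# Each update negates head and neck deltas from reference pose, scaling rotation per axis, and maps PlayerMovement output into left/right values.
# The AnimGraph feeds those values to the Control Rig, then additively applies the counter-transforms to head, neck_02 and neck_01, enabled only with a valid pawn owner.
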AnimBlueprint /BRCosmetics/Characters/Player/Male/Medium/Heads/M_MED_PixieParts_Lava_Head/Meshes/OBSOLETE/M_MED_PixieParts_Lava_Head_PostProcess_AnimBP
target_skeleton=&/Fortnite_Base_Head_PixieParts_Lava_Skeleton
graphs [
 AnimGraph {nodes [
   n1 = Control Rig!(
    Source=Input Pose()
    Alpha
    Character Move Right:double=Get Character Move Right() -> (Character Move Right:double)
    Character Move Left:double=Get Character Move Left() -> (Character Move Left:double)
   ) -> (Pose) {controlRigAssetReference {blueprintRigClass=&/M_MED_PixieParts_Lava_Head_CtrlRig.M_MED_PixieParts_Lava_Head_CtrlRig_C}}
   n2 = Get Head Rot Alpha() -> (Head Rot Alpha:double)
   n3 = "Transform (Modify) Bone - Bone: head"(
    ComponentPose=Local To Component(LocalPose=@n1)
    Alpha=@n2
    Translation=Get Head Trans() -> (Head Trans:Vector)
    Rotation=Get head rot() -> (head rot:Rotator)
    Scale="1,1,1"
   ) {
    rotationMode=BMM_Additive
    rotationSpace=BCS_BoneSpace
    translationMode=BMM_Additive
    translationSpace=BCS_BoneSpace
    boneToModify {boneName=head}
    rotation=<rot 50 50 50>
   }
   n4 = "Transform (Modify) Bone - Bone: neck_01"(
    ComponentPose="Transform (Modify) Bone - Bone: neck_02"(ComponentPose=@n3, Alpha=@n2, Translation="0,0,0", Rotation=Get Neck_02_Rot() -> (Neck_02_Rot:Rotator), Scale="1,1,1") {rotationMode=BMM_Additive, rotationSpace=BCS_BoneSpace, boneToModify {boneName=neck_02}}
    Alpha=@n2
    Translation=Get Neck 01 Trans() -> (Neck 01 Trans:Vector)
    Rotation=Get Neck_01_Rot() -> (Neck_01_Rot:Rotator)
    Scale="1,1,1"
   ) {
    rotationMode=BMM_Additive
    rotationSpace=BCS_BoneSpace
    translationMode=BMM_Additive
    translationSpace=BCS_BoneSpace
    boneToModify {boneName=neck_01}
   }
   Output Pose(Result=Component To Local(ComponentPose=@n4))
  ]}
 EventGraph {nodes [
   Event BlueprintInitializeAnimation()
   > n1 = Cast To FortPlayerPawn(Object=TryGetFortPawn())
   > Set Pawn(Pawn:Fort Player Pawn Object Reference=@n1) -> (Output_Get:Fort Player Pawn Object Reference)
   Event BlueprintUpdateAnimation()
   n2 = BreakTransform(InTransform=GetDeltaTransformFromRefPose(self=GetOwnerSkeletalMeshComponent(), BoneName=neck_01, BaseName=spine_05))
   n3 = BreakRotator(InRot=@n2.Rotation)
   n4 = Get Y percentage() -> (Y percentage:float)
   n5 = Get X percentage() -> (X percentage:float)
   n6 = Get Z percentage() -> (Z percentage:float)
   > Set Neck_01_Rot(Neck_01_Rot:Rotator=MakeRotator(
     Roll=float * float!(A=@n3, B=-1, C:double=@n5) -> (ReturnValue)
     Pitch=float * float(A=@n3.Pitch, B=-1, C=@n4)
     Yaw=float * float(A=@n3.Yaw, B=-1, C=@n6)
    )) -> (Output_Get:Rotator)
   n7 = BreakVector(InVec=@n2)
   > Set Neck 01 Trans(Neck 01 Trans:Vector=MakeVector(X=float * float(A=@n7, B=-1) {is=title K2Node_PromotableOperator}, Y=float * float(A=@n7.Y, B=-1), Z=float * float(A=@n7.Z, B=-1))) -> (Output_Get:Vector)
   n8 = GetOwnerSkeletalMeshComponent()
   n9 = BreakRotator(InRot=BreakTransform(InTransform=GetDeltaTransformFromRefPose(self=@n8, BoneName=neck_02, BaseName=neck_01)).Rotation)
   n10 = Get Z percentage()
   n11 = Get X percentage()
   n12 = Get Y percentage()
   > Set Neck_02_Rot(Neck_02_Rot:Rotator=MakeRotator(
     Roll=float * float(A=@n9, B=-1, C=@n11) {is=title K2Node_CommutativeAssociativeBinaryOperator}
     Pitch=float * float(A=@n9.Pitch, B=-1, C=@n12)
     Yaw=float * float(A=@n9.Yaw, B=-1, C=@n10)
    )) -> (Output_Get:Rotator)
   n13 = BreakTransform(InTransform=GetDeltaTransformFromRefPose(self=GetOwnerSkeletalMeshComponent(), BoneName=head, BaseName=neck_02))
   n14 = BreakRotator(InRot=@n13.Rotation)
   n15 = Get Z percentage()
   n16 = Get X percentage()
   n17 = Get Y percentage()
   > Set head rot(head rot:Rotator=MakeRotator(Roll=float * float(A=@n14, B=-1, C=@n16), Pitch=float * float(A=@n14.Pitch, B=-1, C=@n17), Yaw=float * float(A=@n14.Yaw, B=-1, C=@n15))) -> (Output_Get:Rotator)
   n18 = BreakVector(InVec=@n13)
   > Set Head Trans(Head Trans:Vector=MakeVector(X=float * float(A=@n18, B=-1) {is=title K2Node_PromotableOperator}, Y=float * float(A=@n18.Y, B=-1), Z=float * float(A=@n18.Z, B=-1))) -> (Output_Get:Vector)
   > n19 = Is Valid(InputObject=TryGetPawnOwner())
   Set Head Rot Alpha(execute=@n19.Is Not Valid, Head Rot Alpha:double=0) -> (Output_Get:double)
   Set Head Rot Alpha(execute=@n19.Is Valid, Head Rot Alpha=1)
   > Get!(execute:Exec) -> (then:Exec, else:Exec, Pawn:Fort Player Pawn Object Reference)
   > n20 = PlayerMovement(AnimInstance=Self-Reference(), XDotProductCoefficient=25, YDotProductCoefficient=25, ZDotProductCoefficient=1)
   > Set Character Move Right(Character Move Right:double=MapRangeClamped(Value=@n20, InRangeA=0, InRangeB=25)) -> (Output_Get:double)
   > Set Character Move Left(Character Move Left:double=MapRangeClamped(Value=@n20, InRangeA=0, InRangeB=-25)) -> (Output_Get:double)
  ]}
]
variables [
 Character Move Left {type=number}
 Character Move Right {type=number}
 Head Rot {type=Rotator}
 Head Rot Alpha {type=number}
 Head Trans {type=Vector}
 Neck 01 Trans {type=Vector}
 Neck_01_Rot {type=Rotator}
 Neck_02_Rot {type=Rotator}
 Pawn {type=FortPlayerPawn}
 X Percentage {type=number, default=0.5}
 Y Percentage {type=number, default=1}
 Z Percentage {type=number, default=0.75}
]
